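# Looping SoundCue for the IO launch pad's idle sound in the Athena game mode.
# A single looping wave player feeds the output at reduced volume, routed through the Creative gadgets sound class.
# It borrows the DirectionalLauncher ambient attenuation and overrides concurrency to block new instances rather than stop playing ones once the limit is reached.
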SoundCue /Game/Athena/Sounds/LaunchPad/IO/LaunchPad_IO_Loop_01_Cue
root=WavePlayer_1
nodes [{id=WavePlayer_1, properties {bLooping=true, soundWaveAssetPtr=&/LaunchPad_IO_Loop_01}}]
properties {
 attenuationSettings=/Game/Athena/Sounds/LaunchPad/DirectionalLauncher/DirectionalLauncher_Ambient_Att
 bOverrideConcurrency=true
 maxDistance=2700
 soundClassObject=&Fort_Gadgets_Creative
 volumeMultiplier=0.5
 concurrencyOverrides {resolutionRule=PreventNew}
}
