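# Sound cue for the first-person launch sound of the IO variant of the launch pad, routed to the Fort_Traps sound class.
# A random node picks one of three equally weighted wave variations and feeds a modulator that slightly varies pitch on each play.
# The cue raises base volume and pitch, and its concurrency override allows a single instance, refusing new plays while one is active.
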SoundCue /Game/Athena/Sounds/LaunchPad/IO/LaunchPad_Launch_IO_1P_Cue
root=Modulator_3
nodes [
 {id=Modulator_3, children [Random_2], properties {pitchMax=1.1, pitchMin=0.9}}
 {id=Random_2, children [WavePlayer_5,WavePlayer_7,WavePlayer_8], properties {weights [1,1,1]}}
 {id=WavePlayer_5, properties {soundWaveAssetPtr=&/LaunchPad_IO_1P_01}}
 {id=WavePlayer_7, properties {soundWaveAssetPtr=&/LaunchPad_IO_1P_02}}
 {id=WavePlayer_8, properties {soundWaveAssetPtr=&/LaunchPad_IO_1P_03}}
]
properties {
 bOverrideConcurrency=true
 maxDistance=2097152
 pitchMultiplier=1.1
 soundClassObject=&Fort_Traps
 volumeMultiplier=1.4
 concurrencyOverrides {maxCount=1, resolutionRule=PreventNew}
}
