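# Third-person launch sound cue for the IO launch pad in Athena.
# A random node picks one of three equally weighted IO launch pad wave variations and feeds it through a modulator at the root.
# The cue uses the shared Fort traps attenuation and sound class, with reduced volume and pitch.
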SoundCue /Game/Athena/Sounds/LaunchPad/IO/LaunchPad_Launch_IO_3P_Cue
root=Modulator_4
nodes [
 {id=Modulator_4, children [Random_3]}
 {id=Random_3, children [WavePlayer_4,WavePlayer_5,WavePlayer_8], properties {weights [1,1,1]}}
 {id=WavePlayer_4, properties {soundWaveAssetPtr=&/LaunchPad_IO_3P_01}}
 {id=WavePlayer_5, properties {soundWaveAssetPtr=&/LaunchPad_IO_3P_02}}
 {id=WavePlayer_8, properties {soundWaveAssetPtr=&/LaunchPad_IO_3P_03}}
]
properties {
 attenuationSettings=&Fort_Traps_Att
 maxDistance=6000
 pitchMultiplier=0.9
 soundClassObject=&Fort_Traps
 volumeMultiplier=0.7
}
